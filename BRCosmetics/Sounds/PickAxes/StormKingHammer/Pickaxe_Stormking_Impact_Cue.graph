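# Impact sound cue for the Storm King Hammer pickaxe cosmetic, routed to the impacts sound class with shared pickaxe attenuation.
# A short fixed delay precedes a mixer that layers one of six random Stormking impact waves with one of four generic non-spatialized player flesh impacts.
# The flesh layer sits slightly lower through mixer input volume and random volume modulation.
# Overridden concurrency caps instances per owner, stopping the oldest.
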SoundCue /BRCosmetics/Sounds/PickAxes/StormKingHammer/Pickaxe_Stormking_Impact_Cue
root=Delay_1
nodes [
 {id=Delay_1, children [Mixer_0], properties {delayMax=0.1, delayMin=0.1}}
 {id=Mixer_0, children [Random_1,Modulator_2], properties {inputVolume [1,0.9]}}
 {id=Modulator_2, children [Random_3], properties {volumeMax=0.9, volumeMin=0.8}}
 {
  id=Random_1
  children [WavePlayer_4, WavePlayer_5, WavePlayer_6, WavePlayer_7, WavePlayer_14, WavePlayer_15]
  properties {weights [1,1,1,1,1,1]}
 }
 {id=Random_3, children [WavePlayer_16, WavePlayer_17, WavePlayer_18, WavePlayer_19], properties {weights [1,1,1,1]}}
 {id=WavePlayer_14, properties {soundWaveAssetPtr=&/Pickaxe_Stormking_ImpactB_05}}
 {id=WavePlayer_15, properties {soundWaveAssetPtr=&/Pickaxe_Stormking_ImpactB_06}}
 {id=WavePlayer_16, properties {soundWaveAssetPtr=/Game/Sounds/Fort_Impact_Sounds/Flesh/new_player_impact_nonspat_A_01n}}
 {id=WavePlayer_17, properties {soundWaveAssetPtr=/Game/Sounds/Fort_Impact_Sounds/Flesh/new_player_impact_nonspat_A_02n}}
 {id=WavePlayer_18, properties {soundWaveAssetPtr=/Game/Sounds/Fort_Impact_Sounds/Flesh/new_player_impact_nonspat_A_03n}}
 {id=WavePlayer_19, properties {soundWaveAssetPtr=/Game/Sounds/Fort_Impact_Sounds/Flesh/new_player_impact_nonspat_A_04n}}
 {id=WavePlayer_4, properties {soundWaveAssetPtr=&/Pickaxe_Stormking_ImpactB_01}}
 {id=WavePlayer_5, properties {soundWaveAssetPtr=&/Pickaxe_Stormking_ImpactB_02}}
 {id=WavePlayer_6, properties {soundWaveAssetPtr=&/Pickaxe_Stormking_ImpactB_03}}
 {id=WavePlayer_7, properties {soundWaveAssetPtr=&/Pickaxe_Stormking_ImpactB_04}}
]
properties {
 attenuationSettings=/Game/Sounds/Attenuation/Fort_Pickaxe_Impact_Att
 audioPropertiesSheet=/BRCosmetics/Sounds/Settings/PropertySheets/Pickaxe/APS_Cosmetic_Pickaxe_Impact
 bOverrideConcurrency=true
 bPrimeOnLoad=true
 maxDistance=4300
 soundClassObject=&Fort_Impacts
 volumeMultiplier=0.525
 attenuationOverrides {
  bAttenuate=false
  bAttenuateWithLPF=true
  distanceAlgorithm=Logarithmic
  falloffDistance=800
  lPFRadiusMax=800
  lPFRadiusMin=400
  attenuationShapeExtents {x=300}
 }
 concurrencyOverrides {bLimitToOwner=true, maxCount=3, resolutionRule=StopOldest}
 concurrencySet [/Game/Sounds/Concurrency_Rules/OneMaximum_Owner_StopOldest]
}
